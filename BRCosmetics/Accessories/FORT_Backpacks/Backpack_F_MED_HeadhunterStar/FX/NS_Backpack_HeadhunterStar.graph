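# Niagara effect for the HeadhunterStar backpack cosmetic that adds glowing lights at two thruster points.
# Its single Thruster_Light emitter bursts two particles, using each partition index to pick a skeletal mesh bone or socket and an air-glow offset.
# A system light color feeds light, air-glow and decal-glow values for light, sprite and component renderers.
# Light intensity varies by time of day, with front-end modifiers boosting daytime intensity and decal glow.
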NiagaraSystem /BRCosmetics/Accessories/FORT_Backpacks/Backpack_F_MED_HeadhunterStar/FX/NS_Backpack_HeadhunterStar
dependencies {
 usedDataInterfaces [&NiagaraDataInterfaceArrayFloat3]
 usedDynamicInputs []
 usedModules [
  &self:SystemSpawnScript.SystemScriptSource.SystemScriptGraph.NiagaraNodeAssignment_0.SetVariables_#1
  &SystemState
 ]
}
emitters {Thruster_Light {
  emitterUpdateScript [
   &EmitterState(Life Cycle Mode=::NewEnumerator1, Loop Duration=1, UseLoopDelay=false)
   &SpawnBurst_Instantaneous(Spawn Count=2, Use Spawn Probability=false, UseLoopCountLimit=false)
  ]
  particleSpawnScript [
   &PartitionParticles(Particles Per Partition=1, Partition Mode:ENiagara_PartitionMode=::NewEnumerator1)
   &InitializeParticle(
    Color Mode=::NewEnumerator0
    Position Offset=<50 0 0>
    Sprite Size Mode=::NewEnumerator3
    Uniform Sprite Size=15
    UsePositionOffset=false
    Write Material Random=true
   ) {module=InitializeParticle001}
   &SkeletalMeshLocation(
    Bone / Socket Index=$Particles.PartitionParticles.Partition
    Interpolate Mesh Sampling=false
    MaskSpawnGroup=false
    Mesh Sampling Type=::NewEnumerator5
    Skeletal Mesh=$User.SkeletalMesh
    Socket Sampling Mode=::NewEnumerator5
    Use Sampled Mesh Attribute as Ribbon ID=false
   ) {module=SkeletalMeshLocation001}
   /SetVariables(
    Particles.LightColor=&MultiplyLinearColor(Linear Color=$System.LightColor, Scale Factor=&MakeVector4_FromFloat(Float=10))
    Particles.LightComponentIntensity:NiagaraFloat=&FN_Select_Float_from_Time_Of_Day(Morning=6, Day=&FN_FrontEndFloatModifier(Gameplay=6, Front End=18), Evening=3, Night=3)
    Particles.LightExponent=3
    Particles.LightRadius=18
   ) {isSetParameters=true}
   /SetVariables(
    Particles.AirGlowColor:LinearColor=&MultiplyLinearColor(Linear Color=$Particles.LightColor, Scale Factor=&MakeVector4_FromFloat(Float=0.02))
    Particles.AirGlowPosition:NiagaraPosition=&AddVectorToPosition(
     Position=$Particles.Position
     Vector=&SelectVectorFromArray(
      Direct Array Index=$Particles.PartitionParticles.Partition
      Vector Selection Array=DataInterface() {propertyValues {FloatData [<-4 0 0>, <4 0 0>]}}
     )
    )
    Particles.AirGlowSpriteScale:Vector2f=&Multiply_Vector2D_ByFloat(Float=60, Vector2D=<1 1>)
   ) {isSetParameters=true}
   /SetVariables(
    Particles.DecalGlowColor:LinearColor=&MultiplyLinearColor(
     Linear Color=$Particles.LightColor
     Scale Factor=&MakeVector4_FromFloat(Float=&FN_FrontEndFloatModifier(Gameplay=0.01, Front End=0.1))
    )
    Particles.DecalGlowSpriteScale:Vector2f=&Multiply_Vector2D_ByFloat(Float=15, Vector2D=<1 1>)
   ) {isSetParameters=true}
  ]
  particleUpdateScript [&ParticleState()]
  rendererClasses [
   &NiagaraLightRendererProperties
   &NiagaraSpriteRendererProperties
   &NiagaraComponentRendererProperties
  ]
  renderers [
   {rendererClass=&NiagaraLightRendererProperties, rendererIndex=0}
   {rendererClass=&NiagaraSpriteRendererProperties, rendererIndex=1}
   {rendererClass=&NiagaraSpriteRendererProperties, rendererIndex=2}
   {rendererClass=&NiagaraComponentRendererProperties, rendererIndex=3}
   {rendererClass=&NiagaraSpriteRendererProperties, rendererIndex=4}
  ]
 }}
summary {
 systemName=NS_Backpack_HeadhunterStar
 emitters [{
   emitterName=Thruster_Light
   rendererClasses [
    &NiagaraLightRendererProperties
    &NiagaraSpriteRendererProperties
    &NiagaraComponentRendererProperties
   ]
  }]
 userVariables [
  User.bIsFrontEnd {type=NiagaraBool}
  User.bIsFrontEndPreview {type=NiagaraBool}
  User.SkeletalMesh {
   type=NiagaraDataInterfaceSkeletalMesh
   defaultValue {
    struct=NiagaraExt_VariableValue_DataInterface
    value {dataInterface=&self:NiagaraDataInterfaceSkeletalMesh_0}
   }
  }
  User.SkeletalMeshLights {
   type=NiagaraDataInterfaceSkeletalMesh
   defaultValue {
    struct=NiagaraExt_VariableValue_DataInterface
    value {dataInterface=&self:NiagaraDataInterfaceSkeletalMesh_1}
   }
  }
 ]
}
