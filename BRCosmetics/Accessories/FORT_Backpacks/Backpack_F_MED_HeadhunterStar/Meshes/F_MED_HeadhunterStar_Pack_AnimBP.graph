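# Animation Blueprint for the HeadhunterStar backpack cosmetic, adding secondary physics motion to its handle bones.
# Two chained Anim Dynamics nodes with gravity overrides drive the base handle as a limited angular spring, then the handle as a linear spring sliding narrowly along X.
# The second node's alpha is Movement: zero without 2D movement, otherwise an interpolated value that drops partway while jumping or falling.
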AnimBlueprint /BRCosmetics/Accessories/FORT_Backpacks/Backpack_F_MED_HeadhunterStar/Meshes/F_MED_HeadhunterStar_Pack_AnimBP
target_skeleton=&/F_MED_HeadhunterStar_Pack_Skeleton
graphs [
 AnimGraph {nodes [
   n1 = Anim Dynamics - Bone: dyn_base_handle!(ComponentPose=Local To Component(LocalPose=Input Pose()), Alpha=1, GravityOverride:Vector="0,0,0", ExternalForce="0,0,0") -> (Pose) {
    angularBiasOverride=0.999
    angularDampingOverride=0.965
    angularSpringConstant=150
    bAngularSpring=true
    bOverrideAngularDamping=true
    bUseGravityOverride=true
    linearDampingOverride=0.99
    numSolverIterationsPostUpdate=4
    numSolverIterationsPreUpdate=16
    alphaBoolBlend {blendOption=Linear}
    boundBone {boneName=dyn_base_handle}
    componentAppliedLinearAccClamp=<100 100 100>
    componentLinearAccScale=<0.1 0.1 0.1>
    componentLinearVelScale=<0.1 0.2 0.1>
    gravityOverride {z=800}
    physicsBodyDefinitions [{
      boundBone {boneName=dyn_base_handle}
      boxExtents=<5 5 5>
      constraintSetup {angularLimitsMax=<10 0 30>, angularLimitsMin=<-10 0 -25>, angularTarget=<1 0 0>}
      localJointOffset=<0.8 0 0>
     }]
   }
   n2 = Anim Dynamics - Bone: dyn_handle!(ComponentPose=@n1, Alpha=Get Movement() -> (Movement:double), GravityOverride:Vector="0,0,0", ExternalForce="0,0,0") -> (Pose) {
    angularBiasOverride=0.999
    angularDampingOverride=0.9
    angularSpringConstant=100
    bLinearSpring=true
    bOverrideLinearDamping=true
    bUseGravityOverride=true
    linearDampingOverride=0.9
    linearSpringConstant=100
    numSolverIterationsPostUpdate=4
    numSolverIterationsPreUpdate=16
    alphaBoolBlend {blendInTime=1, blendOption=Linear, blendOutTime=1}
    boundBone {boneName=dyn_handle}
    componentAppliedLinearAccClamp=<300 300 300>
    componentLinearAccScale=<0.1 0.1 0.1>
    componentLinearVelScale=<0.1 0.1 0.1>
    gravityOverride {z=800}
    physicsBodyDefinitions [{
      boundBone {boneName=dyn_handle}
      boxExtents=<7 7 7>
      constraintSetup {angularLimitsMax=<0 0 0>, angularLimitsMin=<0 0 0>, angularTarget=<1 0 0>, linearAxesMax=<0.35 0 0>, linearAxesMin=<-0.35 0 0>}
      localJointOffset=<0 0 0>
     }]
   }
   Output Pose(Result=Component To Local(ComponentPose=@n2)) {layerGroup=None}
  ]}
 EventGraph {nodes [
   n1 = Event BlueprintUpdateAnimation()
   n2 = OR Boolean(A=Get bIsJumping(), B=Get bIsFalling())
   > Set AlphaJump(AlphaJump:float=FInterpTo(Current=Get AlphaJump() -> (AlphaJump:float), Target=SelectFloat(A=0.4, B=1, bPickA=@n2), DeltaTime=@n1.DeltaTimeX, InterpSpeed=5)) -> (Output_Get:float)
   n3 = Get bIsMoving2D()
   > Set Movement(Movement:double=float * float(A=Get AlphaJump(), B="To Float (Boolean)"(InBool=@n3)) {is=title K2Node_PromotableOperator}) -> (Output_Get:double)
   TryGetPawnOwner()
  ]}
]
variables [
 AlphaJump {type=number}
 Movement {type=number}
]
